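# Material function that layers an optional anodized, angle-dependent color effect onto incoming material attributes.
# A Fresnel term offset by OpacityMask height samples a gradient LUT, brightened at sun reflections, to tint BaseColor and optionally adjust Metallic.
# HeightLerp blends the result using per-layer weights, OpacityMask and an optional mesh mask, limited by metalness.
# A static switch toggles it on the default shading path, with Forward and Mobile routed separately.
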MaterialFunction /Game/Athena/WinterArt/Materials/MF_Anodizing
graph {nodes [
  n1 = Constant() {r=1}
  n2 = Constant() {r=3}
  n3 = FunctionInput() {inputType=FunctionInput_MaterialAttributes}
  n4 = GetMaterialAttributes(None=NamedRerouteDeclaration(None=@n3) {name=AnodizingMatInput, nodeColor=<rgba 0.0312 1 0 1>}) {attributeGetTypes [
    #1
    #2
    #3
    #4
   ]}
  n5 = Transform(None=@n4.Normal)
  n6 = ScalarParameter() {group=Anodizing, parameterName=Anodizing Height To FX}
  n7 = TextureSampleParameter2D(UVs=Add(A=Fresnel(ExponentIn=@n1, BaseReflectFractionIn=@n2, Normal=@n5) {baseReflectFraction=0.5, exponent=3}, B=Multiply(A=@n4.OpacityMask, B=@n6))) {
   group=Anodizing
   parameterName=IrradianceCurveLUT
   samplerSource=SSM_Clamp_WorldGroupSettings
   texture=/Game/Global/Textures/Default/Masks_and_Gradients/3pxRGBImage
  }
  n8 = Multiply(A=Constant() {r=4}, B=@n7.RGB)
  n9 = DotProduct(A=AtmosphericLightVector(), B=ReflectionVectorWS())
  n10 = Constant() {r=55}
  n11 = Saturate(None=Multiply(A=Power(Base=@n9, Exp=@n10), B=Constant() {r=512}))
  n12 = Multiply(
   A=ScalarParameter() {defaultValue=1, group=Anodizing, parameterName=Anodizing, sortPriority=1}
   B=Multiply(A=LinearInterpolate(A=@n7.RGB, B=@n8, Alpha=@n11), B=@n4.BaseColor)
  )
  n13 = Fresnel(Normal=@n5) {exponent=3}
  n14 = Add(A=@n4.Metallic, B=Multiply(A=Constant() {r=0.8}, B=OneMinus(None=@n13)))
  n15 = OneMinus(None=ScalarParameter() {group=Anodizing, parameterName="Anodizing_MetalMask "})
  n16 = ScalarParameter() {defaultValue=0.5, group=Anodizing, parameterName=Anodizing Transition, sortPriority=2}
  n17 = Multiply(
   A=@n4.OpacityMask
   B=MaterialFunctionCall!(
    transitionFactor
    1=ScalarParameter() {defaultValue=1, group=Anodizing, parameterName=L1_AnodizingMask, sortPriority=4}
    2=ScalarParameter() {defaultValue=1, group=Anodizing, parameterName=L2_AnodizingMask, sortPriority=5}
    3=ScalarParameter() {defaultValue=1, group=Anodizing, parameterName=L3_AnodizingMask, sortPriority=6}
    4=ScalarParameter() {defaultValue=1, group=Anodizing, parameterName=L4_AnodizingMask, sortPriority=7}
    Use 2 Textures=StaticBoolParameter() {group=Material_2, parameterName=Use 2 Materials}
    Use 3 Textures=StaticBoolParameter() {group=Material_3, parameterName=Use 3 Materials}
    Use 4 Textures=StaticBoolParameter() {group=Material_4, parameterName=Use 4 Materials}
   ) {materialFunction=&/MF_LerpTexturesBasedOnInt_Float4_4Layers}.Result
  )
  n18 = TextureCoordinate()
  n19 = TextureCoordinate() {coordinateIndex=2}
  n20 = TextureSampleParameter2D(UVs=StaticSwitchParameter(
    True=TextureCoordinate() {coordinateIndex=3}
    False=StaticSwitchParameter(True=@n18, False=@n19) {desc=Uses your Height maps as the mask, group=Anodizing, parameterName=Anodizing Mask UV0, sortPriority=11}
   ) {desc=Uses your Height maps as the mask, group=Anodizing, parameterName=Anodizing Mask UV3, sortPriority=11}) {
   group=Anodizing
   parameterName="Anodizing Mesh mask (UV2)"
   samplerSource=SSM_Clamp_WorldGroupSettings
   samplerType=SAMPLERTYPE_Masks
   sortPriority=9
   texture=/Game/Athena/WinterArt/Textures/T_AnimMask
  }
  n21 = StaticComponentMaskParameter(None=@n20.RGB) {
   defaultR=true
   desc=You can use some shared texture between multiple functions by selecting different channels and UV's
   group=Anodizing
   parameterName=Anodizing Mesh Mask Channel
   sortPriority=11
  }
  n22 = StaticSwitchParameter(True=Max(A=@n17, B=@n21), False=@n17) {desc=Uses ycustom texture here with tile heights, group=Anodizing, parameterName=Anodizing Mask Texture true, sortPriority=9}
  n23 = ScalarParameter() {group=Anodizing, parameterName=Anodizing Transition Contrast, sortPriority=3}
  n24 = BlendMaterialAttributes(
   A=@n4.MaterialAttributes
   B=SetMaterialAttributes!(
    MaterialAttributes=@n4.MaterialAttributes
    BaseColor=Add(A=@n12, B=@n4.BaseColor)
    Metallic=StaticSwitchParameter(True=@n14, False=@n4.Metallic) {defaultValue=true, group=Anodizing, parameterName=EnableMetallicAnodizing}
   ) {attributeSetTypes [#1, #4]}
   Alpha=Multiply(
    A=Max(A=@n4.Metallic, B=@n15)
    B=MaterialFunctionCall(Transition Phase=@n16, Height Texture=@n22, Contrast=@n23) {materialFunction=&HeightLerp}.Alpha
   )
  ) {vertexAttributeBlendType=UseA}
  n25 = NamedRerouteUsage()
  FunctionOutput(None=ShadingPathSwitch(
    Default=StaticSwitchParameter(True=@n24, False=@n4.MaterialAttributes) {desc=Uses your Height maps as the mask, parameterName=Anodizing, sortPriority=-2}
    Forward=@n25
    Mobile=@n25
   ))
  Reroute(None=@n3)
 ]}
