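# Material function for WinterArt materials that boosts a material's emissive color in response to the event audio spectrum.
# It dots the spectrum from FortniteMaterialParameters with a band mask, then applies an exponent, bias and low/high remap.
# The result scales the incoming emissive, is clamped at zero, weighted by an optional EmissiveMask input and added back.
# A static switch passes the attributes through unchanged when the effect is off.
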MaterialFunction /Game/Athena/WinterArt/Materials/MF_SoundToEmissive
graph {nodes [
  n1 = FunctionInput() {inputType=FunctionInput_MaterialAttributes}
  n2 = GetMaterialAttributes(None=@n1) {attributeGetTypes [#1]}
  n3 = CollectionParameter() {collection=&FortniteMaterialParameters, parameterName=EventAudio_Spectrum}
  n4 = VectorParameter() {group=SoundToEmissive, parameterName=EventAudioSpectrumMask, defaultValue=<rgba 1 0 0 0>}
  n5 = Power(
   Base=DotProduct(A=@n3, B=AppendVector(A=@n4.RGB, B=@n4.RGB))
   Exp=ScalarParameter() {defaultValue=1.5, group=SoundToEmissive, parameterName=EventAudioSpectrumExponent}
  )
  n6 = ScalarParameter() {group=SoundToEmissive, parameterName=EventAudioSpectrumBias}
  n7 = Multiply(
   A=@n2.EmissiveColor
   B=LinearInterpolate(
    A=ScalarParameter() {group=SoundToEmissive, parameterName=EventAudioSpectrumRemapLow}
    B=ScalarParameter() {defaultValue=1, group=SoundToEmissive, parameterName=EventAudioSpectrumRemapHigh}
    Alpha=Add(A=@n5, B=@n6)
   )
  )
  n8 = FunctionInput() {bUsePreviewValueAsDefault=true, inputName=EmissiveMask, inputType=FunctionInput_Scalar, sortPriority=2, previewValue=<1 1 1 1>}
  n9 = Add(A=@n2.EmissiveColor, B=Multiply(A=Max(A=@n7) {constB=0}, B=@n8))
  FunctionOutput(None=StaticSwitchParameter(
    True=SetMaterialAttributes!(MaterialAttributes=@n2.MaterialAttributes, EmissiveColor=@n9) {attributeSetTypes [#1]}
    False=@n1
   ) {group=SoundToEmissive, parameterName=EventAudio to Emissive, sortPriority=0})
 ]}
