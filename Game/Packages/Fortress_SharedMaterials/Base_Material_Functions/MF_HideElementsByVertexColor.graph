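# Material function that hides parts of a mesh selected by vertex color, operating on incoming material attributes.
# It sums vertex RGBA multiplied by the "Elements to Hide by color" vector parameter, red by default, into a mask.
# That mask scales an offset toward the object position, added to the existing World Position Offset to collapse selected vertices.
# A static switch, HideElementByVertexColor, otherwise passes the attributes through unchanged.
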MaterialFunction /Game/Packages/Fortress_SharedMaterials/Base_Material_Functions/MF_HideElementsByVertexColor
graph {nodes [
  n1 = FunctionInput() {inputType=FunctionInput_MaterialAttributes}
  n2 = GetMaterialAttributes(None=@n1) {attributeGetTypes [#1]}
  n3 = ObjectPositionWS()
  n4 = Add(A=@n2.WorldPositionOffset, B=WorldPosition().XYZ)
  n5 = VertexColor()
  n6 = VectorParameter() {group=Hide Elements by VertexColor, parameterName=Elements to Hide by color, defaultValue=<rgba 1 0 0 0>}
  n7 = Add(A=Add(A=Multiply(A=@n5.R, B=@n6.R), B=Multiply(A=@n5.G, B=@n6.G)), B=Multiply(A=@n5.B, B=@n6.B))
  n8 = Multiply(A=@n5.A, B=@n6.A)
  n9 = Add(A=@n2.WorldPositionOffset, B=Multiply(A=Subtract(A=@n3, B=@n4), B=Add(A=@n7, B=@n8)))
  FunctionOutput(None=StaticSwitchParameter(
    True=SetMaterialAttributes!(MaterialAttributes=@n2.MaterialAttributes, WorldPositionOffset=@n9) {attributeSetTypes [#1]}
    False=@n1
   ) {group=Hide Elements by VertexColor, parameterName=HideElementByVertexColor})
 ]}
