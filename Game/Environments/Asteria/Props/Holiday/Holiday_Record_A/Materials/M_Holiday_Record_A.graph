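# Layered master-style prop material for the Holiday Record A environment asset, with tintable textures, optional two-sided maps, emissive, clearcoat and subsurface parameters.
# A four-layer blend function, fed mainly by layer one, passes through chained effect functions including dirt, gradients, screens and optional moss.
# A shading-path switch gives Forward and Mobile shorter chains before shared snow and game-wide functions.
# World position offset branches add sine shake, or CozyLodge animation with a Fadeout collapse toward object position.
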Material /Game/Environments/Asteria/Props/Holiday/Holiday_Record_A/Materials/M_Holiday_Record_A
graph {
 properties {
  bAutomaticallySetUsageInEditor=false
  bUseMaterialAttributes=true
  bUsedWithNanite=true
  ditheredLODTransition=true
  pixelDepthOffsetMode=PDOM_Legacy
  refractionMethod=RM_IndexOfRefraction
  usedShadingModels=Default Lit
 }
 nodes [
  n1 = StaticBool() {value=true}
  n2 = ScalarParameter() {defaultValue=0.5, group=SSS, parameterName=SSSDepth}
  n3 = MaterialFunctionCall() {materialFunction=/Game/Athena/WinterArt/Materials/MF_UVsSet}
  n4 = VectorParameter() {
   group=Material_1
   parameterName=L1_UVScaleByComponent
   sortPriority=3
   channelNames {a="", b=Normals, g=Specular, r=Diffuse}
   defaultValue=<rgba 1 1 1 0>
  }
  n5 = TextureSampleParameter2D(UVs=Multiply(A=@n3.UVs Layer 1, B=@n4.Diffuse)) {
   group=Material_1
   parameterName=Diffuse Side 2
   samplerSource=SSM_Wrap_WorldGroupSettings
   sortPriority=1
   texture=/Game/Global/Textures/Default/Blanks/T_MidgreysRGB
  }
  n6 = TwoSidedSign()
  n7 = VectorParameter() {
   group=Material_1
   parameterName=L1_UVScaleByComponent
   sortPriority=3
   channelNames {a="", b=Normals, g=Specular, r=Diffuse}
   defaultValue=<rgba 1 1 1 0>
  }
  n8 = Multiply(A=@n3.UVs Layer 1, B=@n7.Diffuse)
  n9 = StaticSwitchParameter(True=Multiply(A=ScalarParameter() {defaultValue=1, group=Material_1, parameterName=L1_UV_ColorSwatchScaleAdd}, B=@n8), False=@n8) {group=Material_1, parameterName=Use Color Swatch}
  n10 = TextureSampleParameter2D(UVs=Multiply(A=ScalarParameter() {defaultValue=1, group=Material_1, parameterName=L1_CustomHeightScale, sortPriority=1}, B=@n9)) {
   group=Material_1
   parameterName=L1_Height
   samplerSource=SSM_Wrap_WorldGroupSettings
   samplerType=SAMPLERTYPE_LinearGrayscale
   sortPriority=-5
   texture=/Game/Global/Textures/Default/Blanks/T_BlackLinearGrayscale
  }
  n11 = TextureSampleParameter2D(UVs=@n8) {
   group=Material_1
   parameterName=Background Diffuse
   samplerSource=SSM_Wrap_WorldGroupSettings
   sortPriority=-5
   texture=&T_WhitesRGB
  }
  n12 = TextureSampleParameter2D(UVs=@n9) {
   group=Material_1
   parameterName=Diffuse
   samplerSource=SSM_Wrap_WorldGroupSettings
   sortPriority=-5
   texture=&T_MidgreysRGB
  }
  n13 = StaticSwitchParameter(True=@n10.R, False=StaticSwitchParameter(True=@n11.A, False=@n12.A) {group=Material_1, parameterName=Use Color Swatch}) {desc=Use Smale scale grayscale, group=Material_1, parameterName=L1 Custom Height, sortPriority=1}
  n14 = Multiply(A=@n7.Specular, B=@n3.UVs Layer 1)
  n15 = TextureSampleParameter2D(UVs=@n14) {
   group=Material_1
   parameterName=SpecularMasks
   samplerSource=SSM_Wrap_WorldGroupSettings
   samplerType=SAMPLERTYPE_Masks
   sortPriority=-5
   texture=&T_Fortnite_Default_S
  }
  n16 = MaterialFunctionCall!(
   Transition=ScalarParameter() {defaultValue=0.5, group=Material_1, parameterName=L1_TintTransition, sortPriority=5}
   HM=@n13
   Contrast=ScalarParameter() {defaultValue=5, group=Material_1, parameterName=L1_TintContrast, sortPriority=5}
   Metalness=@n15.G
   BG Color=StaticSwitchParameter(True=@n11.RGB, False=@n12.RGB) {group=Material_1, parameterName=Use Color Swatch}
   TintColor=StaticSwitchParameter(True=@n12.RGB, False=VectorParameter() {group=Material_1, parameterName=TintColor, sortPriority=7, defaultValue=<rgba 1 1 1 0>}.RGB) {group=Material_1, parameterName=Use Color Swatch}
  ) {materialFunction=/Game/Athena/WinterArt/Materials/MF_ColorMask}
  n17 = AppendVector(
   A=StaticSwitchParameter(True=@n16.LerpMix, False=@n16.LerpMix) {group=Material_1, parameterName=L1_UseTintColorAsDiffuse}
   B=Multiply(A=@n13, B=ScalarParameter() {defaultValue=1, group=Material_1, parameterName=L1 Alpha, sortPriority=1})
  )
  n18 = TextureObjectParameter() {group=Material_1, parameterName=Diffuse, sortPriority=1, texture=&T_MidgreysRGB}
  n19 = ScalarParameter() {group=Adv Emissive, parameterName=L1 X Speed, sortPriority=12}
  n20 = ScalarParameter() {group=Adv Emissive, parameterName=L1 Y Speed, sortPriority=13}
  n21 = VectorParameter() {group=Adv Emissive, parameterName=L1 Channel, sortPriority=9}
  n22 = ScalarParameter() {defaultValue=1, group=Adv Emissive, parameterName=L1 Offset, sliderMax=4, sliderMin=1, sortPriority=11}
  n23 = VectorParameter() {
   desc=WARNING* Run through Saturate node when Lumen
   group=Material_1
   parameterName=Emissive
   sortPriority=7
   channelNames {a=Emissive Multiplier, b="", g="", r=""}
   defaultValue=<rgba 0 0 0 1>
  }
  n24 = Multiply(A=@n23, B=@n23)
  n25 = RayTracingQualitySwitch(Normal=@n24, RayTraced=Saturate(None=@n24)) {desc=For Lumen}
  n26 = Multiply(A=@n3.UVs Layer 1, B=@n4.Specular)
  n27 = TextureSampleParameter2D(UVs=@n26) {
   group=Material_1
   parameterName=SpecularMasks Side 2
   samplerSource=SSM_Wrap_WorldGroupSettings
   samplerType=SAMPLERTYPE_Masks
   sortPriority=2
   texture=&T_Fortnite_Default_S
  }
  n28 = TextureCoordinate() {coordinateIndex=1}
  n29 = TextureSampleParameter2D(UVs=StaticSwitchParameter(
    True=TextureCoordinate() {coordinateIndex=2}
    False=StaticSwitchParameter(True=@n28, False=@n26) {group=Material_1, parameterName=L1_Ch2Emissive Side2, sortPriority=-5}
   ) {group=Material_1, parameterName=L1_Ch3Emissive Side2, sortPriority=-5}) {
   group=Material_1
   parameterName=L1_Emissive Side 2
   samplerSource=SSM_Wrap_WorldGroupSettings
   samplerType=SAMPLERTYPE_LinearGrayscale
   sortPriority=-4
   texture=&T_BlackLinearGrayscale
  }
  n30 = AppendVector(A=@n27.RGB, B=StaticSwitchParameter(True=@n29.R, False=@n27.A) {group=Material_1, parameterName=L1_CustomEmissive Side 2, sortPriority=-5})
  n31 = TextureCoordinate() {coordinateIndex=1}
  n32 = TextureSampleParameter2D(UVs=StaticSwitchParameter(
    True=TextureCoordinate() {coordinateIndex=2}
    False=StaticSwitchParameter(True=@n31, False=@n14) {group=Material_1, parameterName=L1_Ch2Emissive, sortPriority=-5}
   ) {group=Material_1, parameterName=L1_Ch3Emissive, sortPriority=1}) {
   group=Material_1
   parameterName=L1_Emissive
   samplerSource=SSM_Wrap_WorldGroupSettings
   samplerType=SAMPLERTYPE_LinearGrayscale
   sortPriority=-5
   texture=&T_BlackLinearGrayscale
  }
  n33 = AppendVector(A=@n15.RGB, B=StaticSwitchParameter(True=@n32.R, False=@n15.A) {group=Material_1, parameterName=L1_CustomEmissive, sortPriority=-5})
  n34 = TextureSampleParameter2D(UVs=Multiply(A=@n4.Normals, B=@n3.UVs Layer 1)) {
   group=Material_1
   parameterName=Normals Side 2
   samplerSource=SSM_Wrap_WorldGroupSettings
   samplerType=SAMPLERTYPE_Normal
   sortPriority=3
   texture=/Engine/EngineMaterials/FlatNormal
  }
  n35 = TextureSampleParameter2D(UVs=Multiply(A=@n7.Normals, B=@n3.UVs Layer 1)) {
   group=Material_1
   parameterName=Normals
   samplerSource=SSM_Wrap_WorldGroupSettings
   samplerType=SAMPLERTYPE_Normal
   sortPriority=-5
   texture=&FlatNormal
  }
  n36 = ScalarParameter() {defaultValue=1, group=Material_1, parameterName=L1_BoostNormals, sortPriority=3}
  n37 = Multiply(A=@n35.RGB, B=AppendVector(A=AppendVector(A=@n36, B=@n36), B=Constant() {r=1}))
  n38 = MaterialFunctionCall!(
   Use3Materials
   Use4Materials
   Use2Materials
   D=StaticSwitchParameter(True=LinearInterpolate(A=@n5.RGBA, B=@n17, Alpha=@n6), False=@n17) {group=Material_1, parameterName=TwoSided, sortPriority=18}
   S=StaticSwitchParameter(True=LinearInterpolate(A=@n30, B=@n33, Alpha=@n6), False=@n33) {group=Material_1, parameterName=TwoSided, sortPriority=18}
   N=StaticSwitchParameter(True=LinearInterpolate(A=@n34.RGB, B=@n37, Alpha=@n6), False=@n37) {group=Material_1, parameterName=TwoSided, sortPriority=18}
   D2=Reroute()
   S2=Reroute()
   N2=Reroute()
   D3
   S3
   N3
   D4
   S4
   N4
   E1=StaticSwitchParameter(
    True=MaterialFunctionCall!(Base Diffuse=@n18) {materialFunction=/Game/Athena/WinterArt/Materials/MF_NeonSignEmissive}.Result
    False=MaterialFunctionCall!(X=@n19, Y=@n20, Dot=@n21.RGB, TextureOffset=@n22, EmissiveTexture=@n25) {materialFunction=/Game/Athena/WinterArt/Materials/MF_EmissiveMask}.EmissiveMasked
   ) {group=NeonEmissive, parameterName=UseNeonSignEmissive_Layer1}
   E2=Reroute()
   E3
   E4
   CCMaskByMetal_1=ScalarParameter() {group=Material_1, parameterName=L1 Metal Mask To Clearcoat, sortPriority=24}
   CC1=ScalarParameter() {group=Material_1, parameterName=L1 Clearcoat, sortPriority=22}
   CCR_1=ScalarParameter() {defaultValue=0.025, group=Material_1, parameterName=L1 Clearcoat Roughness, sortPriority=23}
   CC2
   CCR_2
   CCMaskByMetal_2
   CC3
   CCR_3
   CCMaskByMetal_3
   CC4
   CCMaskByMetal_4
   CCR_4
   OpacityMaskCancel_1=ScalarParameter() {group=Material_1, parameterName=L1_OpacityMaskCancel}
   OpacityMaskCancel_2=ScalarParameter() {group=Material_2, parameterName=L2_OpacityMaskCancel}
   OpacityMaskCancel_3=ScalarParameter() {group=Material_3, parameterName=L3_OpacityMaskCancel}
   OpacityMaskCancel_4=ScalarParameter() {group=Material_4, parameterName=L4_OpacityMaskCancel}
   SSS_1=ScalarParameter() {defaultValue=1, group=SSS, parameterName=L1_SSS}
   SSS_2=ScalarParameter() {defaultValue=1, group=SSS, parameterName=L2_SSS}
   SSS_3=ScalarParameter() {defaultValue=1, group=SSS, parameterName=L3_SSS}
   SSS_4=ScalarParameter() {defaultValue=1, group=SSS, parameterName=L4_SSS}
   SSS_Depth_1=@n2
   SSS_Depth_2=@n2
   SSS_Depth_3=@n2
   SSS_Depth_4=@n2
  ) {materialFunction=/Game/Athena/WinterArt/Materials/MF_FN_TextureSetLerp_4Layers}
  n39 = SetMaterialAttributes!(
   MaterialAttributes
   Base Color=@n38.Diffuse
   Specular=@n38.Diffuse
   Metallic=@n38.Diffuse
   Roughness=@n38.Diffuse
   Emissive Color=@n38.Diffuse
   Normal=@n38.Diffuse
   Custom Data 0=@n38.Diffuse
   Custom Data 1=@n38.Diffuse
   Opacity Mask=@n38.Diffuse
   Subsurface Color=@n38.Diffuse
   Opacity=@n38.Diffuse
   Refraction=NamedRerouteDeclaration(None=@n38.Spec_A) {name=Spec A, nodeColor=<rgba 0 0.2734 1 1>}
  ) {attributeSetTypes [
    #1
    #2
    #3
    #4
    #5
    #6
    #7
    #8
    #9
    #10
    #11
    #12
   ]}
  n40 = MaterialFunctionCall!(In=@n39, Emissive Mask=NamedRerouteUsage()) {materialFunction=/Game/Athena/WinterArt/Materials/MF_EmissiveColor}
  n41 = MaterialFunctionCall!(In=MaterialFunctionCall!(In=@n40.Result, EmissiveMask=@n40.Result) {materialFunction=/Game/Athena/WinterArt/Materials/MF_SoundToEmissive}.Result) {materialFunction=/Game/Athena/WinterArt/Materials/MF_DecalPrint}
  n42 = MaterialFunctionCall!(In=MaterialFunctionCall!(In=@n41.Result) {materialFunction=/Game/Athena/WinterArt/Materials/MF_SphereEmissive}.Result) {materialFunction=/Game/Athena/WinterArt/Materials/MF_TOD_EmissiveColors}
  n43 = MaterialFunctionCall!(In=MaterialFunctionCall!(In=MaterialFunctionCall!(In=@n42.Result) {materialFunction=/Game/Packages/Fortress_SharedMaterials/Base_Material_Functions/MF_HideElementsByVertexColor}.Result) {materialFunction=/Game/Athena/WinterArt/Materials/MF_WS_Texture}.Result) {materialFunction=/Game/Athena/WinterArt/Materials/MF_Anodizing}
  n44 = MaterialFunctionCall!(
   transitionFactor
   MaterialIn=MaterialFunctionCall!(transitionFactor, MaterialIn=MaterialFunctionCall!(In=@n43.Result) {materialFunction=/Game/Athena/WinterArt/Materials/MF_Fuzzy}.Result) {materialFunction=/Game/Athena/WinterArt/Materials/MF_ObjectSpaceGradient}.Color*Gradient
  ) {materialFunction=/Game/Athena/WinterArt/Materials/MF_WSGradient}
  n45 = MaterialFunctionCall!(
   In=MaterialFunctionCall!(
    UV=MaterialFunctionCall() {materialFunction=&MF_UVsSet}.UVs 4 Layers
    In=MaterialFunctionCall!(In=@n44.Color*Gradient, Emissive Mask=@n40.EmissiveMask) {materialFunction=/Game/Athena/WinterArt/Materials/MF_Screens}.Result
   ) {materialFunction=/Game/Athena/WinterArt/Materials/MF_WSDirt}.Result
   UseOneSampler
  ) {materialFunction=/Game/Athena/WinterArt/Materials/MF_VPTexture}
  n46 = MaterialFunctionCall!(In=MaterialFunctionCall!(In=@n45.Result, UpVectorTexture=@n45.Result) {materialFunction=/Game/Athena/WinterArt/Materials/MF_UpVectorTexture}.Result) {materialFunction=/Game/Athena/WinterArt/Materials/M_FakeWindowReflections}
  n47 = StaticSwitchParameter(
   True=MaterialFunctionCall!(WorldSpaceNormals, In=@n46.Result) {materialFunction=/Game/Environments/Apollo/Rocks/Natural_Rocks/Materials/MossSoft2}.Result
   False=@n46.Result
  ) {parameterName=UseMoss}
  n48 = MaterialFunctionCall(In=MaterialFunctionCall(In=@n41.Result, Emissive Mask=@n40.EmissiveMask) {materialFunction=&MF_Screens}.Result) {materialFunction=&MF_VPTexture}
  n49 = MaterialFunctionCall(In=@n48.Result, UpVectorTexture=@n48.Result) {materialFunction=&MF_UpVectorTexture}
  n50 = GetMaterialAttributes(None=@n49.Result) {attributeGetTypes [
    #1
    #5
    #9
    #2
    #4
    #3
   ]}
  n51 = SetMaterialAttributes!(
   MaterialAttributes
   Base Color=@n50.Base Color
   Emissive Color=@n50.Base Color
   Opacity Mask=@n50.Base Color
   Specular=Multiply(A=@n50.Specular, B=ScalarParameter() {defaultValue=1, group=Mobile, parameterName=Mobile_SpecularMult})
   Roughness=Multiply(A=@n50.Roughness, B=ScalarParameter() {defaultValue=1, group=Mobile, parameterName=Mobile_RoughnessMult})
   Metallic=Multiply(A=@n50.Metallic, B=ScalarParameter() {defaultValue=1, group=Mobile, parameterName=Mobile_MetalMult})
   Subsurface Color=@n50.Base Color
  ) {attributeSetTypes [
    #1
    #5
    #9
    #2
    #4
    #3
    #10
   ]}
  n52 = GetMaterialAttributes(None=@n49.Result) {attributeGetTypes [
    #1
    #5
    #9
    #2
    #4
    #3
    #6
   ]}
  n53 = SetMaterialAttributes!(
   MaterialAttributes
   Base Color=@n52.Base Color
   Emissive Color=@n52.Base Color
   Opacity Mask=@n52.Base Color
   Specular=Multiply(A=@n52.Specular, B=ScalarParameter() {defaultValue=1, group=Mobile, parameterName=Mobile_SpecularMult})
   Roughness=Multiply(A=@n52.Roughness, B=ScalarParameter() {defaultValue=1, group=Mobile, parameterName=Mobile_RoughnessMult})
   Metallic=Multiply(A=@n52.Metallic, B=ScalarParameter() {defaultValue=1, group=Mobile, parameterName=Mobile_MetalMult})
   Normal=@n52.Base Color
   Subsurface Color=@n52.Base Color
  ) {attributeSetTypes [
    #1
    #5
    #9
    #2
    #4
    #3
    #6
    #10
   ]}
  n54 = MaterialFunctionCall!(
   IsStaticSnow
   UseAlternateStaticSnowMask
   AlternateStaticSnowMask
   "UseRoad&&OcclusionMask"
   IsForGrassMeshToPushDown
   AdditionalMaskToRemoveSnow
   IsFallingSnowFX
   UseMinimapCoords
   IncreaseSnowUsingSlope
   UseSmoothBorder
   LerpHighM
   TemporalHeightLerpContrast
   IsLobbyIsland
   UseBR-IslandMask=StaticBool()
   IsTree
   IsForRenderTarget
   UseSparkle
   UseSnowNormals=@n1
   UseAdditionalMaskToRemoveSnow
   IsTangentSpaceInput=@n1
   UseVertexNormalsForBasicSnow
   UseTangentSnowNormals=@n1
   Material=ShadingPathSwitch(
    Default=MaterialFunctionCall!(In=@n47) {materialFunction=/Game/Athena/WinterArt/Materials/MF_RandomDecalSplatter}.Result
    Forward=@n49.Result
    Mobile=DistanceFieldsRenderingSwitch!(No=@n51, Yes=@n53)
   )
  ) {
   desc=Realtime
   materialFunction=/Game/Athena/Apollo/Environments/Landscape/Materials/Biome/UniqueBiomeMaterialFunctions/MF_Apollo_Snow_Progression
  }
  n55 = MaterialFunctionCall!(In=@n54."DefaultSwitch=FalseMaterial") {materialFunction=/Game/Athena/WinterArt/Materials/MF_GameWideShared}
  n56 = GetMaterialAttributes(None=@n55.Result) {attributeGetTypes [#13]}
  n57 = Transform(None=VectorParameter() {parameterName=shakedirection, defaultValue=<rgba 0 0 1 1>}.RGB) {transformSourceType=TRANSFORMSOURCE_Local}
  n58 = Multiply(A=Time(), B=ScalarParameter() {defaultValue=3, parameterName=ShakeSpeed})
  n59 = Multiply(A=Multiply(A=Normalize(VectorInput=@n57), B=Sine(None=@n58)), B=ScalarParameter() {defaultValue=3, parameterName=ShakeDist})
  n60 = ScalarParameter() {parameterName=ShakeAmount}
  SetMaterialAttributes!(MaterialAttributes=@n56.MaterialAttributes, World Position Offset=Add(A=@n56.World Position Offset, B=Multiply(A=@n59, B=@n60))) {attributeSetTypes [#13]}
  n61 = GetMaterialAttributes(None=MaterialFunctionCall!(In=@n55.Result) {materialFunction=/Game/Environments/FrontEnd/MaterialFunctions/MF_CozyLodge_Anim}.Result) {attributeGetTypes [#13]}
  n62 = Constant3Vector()
  n63 = Subtract(A=ObjectPositionWS(), B=Add(A=@n61.World Position Offset, B=WorldPosition() {worldPositionShaderOffset=WPT_ExcludeAllShaderOffsets}.XYZ))
  n64 = ScalarParameter() {parameterName=Fadeout}
  SetMaterialAttributes(
   MaterialAttributes=@n61.MaterialAttributes
   World Position Offset=Add(A=@n61.World Position Offset, B=LinearInterpolate(A=@n62, B=@n63, Alpha=@n64))
  ) {attributeSetTypes [#13]}
 ]
}
